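# Parent character material for the CarolinaChili body, adding flipbook face features between the shared 2023 PreFX and PostFX functions.
# An eye-index function processes the PreFX result, then an optional mouth flipbook picks atlas frames by index using selectable UV channels.
# Texture alpha and sphere masks blend in the mouth, with optional emissive and an FX map controlling shading overrides.
# A debug switch offsets vertices by vertex color and remap curve to preview eye poses.
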Material /Game/Characters/Player/Male/Medium/Bodies/M_MED_CarolinaChili/Materials/M_CarolinaChili_Parent
graph {
 properties {
  bHasPixelAnimation=true
  bUseMaterialAttributes=true
  bUsedWithClothing=true
  bUsedWithMeshDeformer=true
  bUsedWithMorphTargets=true
  bUsedWithNanite=true
  bUsedWithNiagaraMeshParticles=true
  bUsedWithSkeletalMesh=true
  pixelDepthOffsetMode=PDOM_Legacy
  shadingModel=MSM_FromMaterialExpression
  usedShadingModels=Default Lit | Subsurface | Clear Coat | Subsurface Profile | Cloth
 }
 nodes [
  n1 = NamedRerouteUsage()
  Floor(None=@n1)
  n2 = NamedRerouteUsage()
  n3 = Constant3Vector() {constant=<rgba 1 1 1 0>}
  n4 = VertexColor()
  n5 = Constant()
  n6 = CurveAtlasRowParameter(CurveTime=Divide(A=ScalarParameter() {defaultValue=0.1, group=Flipbook Eyes, parameterName=DebugEyeIndex}) {constB=10}) {
   atlas=/Game/Characters/Player/Common/MaterialFunctions/VertexColorFaces/CA_FlipbookV2_Remapping
   curve=/Game/Characters/Player/Common/MaterialFunctions/VertexColorFaces/C_EyesAndMouthDefault
   group=Flipbook Eyes
   parameterName=EyeIndexRemapCurve
  }
  n7 = Subtract(
   A=If(A=DotProduct(A=@n3, B=@n4), B=Constant() {r=3}, "A > B"=@n5, "A == B"=@n5, "A < B"=Constant() {r=1})
   B=SphereMask(A=@n6.R, B=VertexColor().G) {attenuationRadius=0.02}
  )
  n8 = Multiply(A=Saturate(None=Round(None=@n7)), B=Constant3Vector() {constant=<rgba 0 0 30 0>}) {constB=30}
  n9 = MaterialFunctionCall() {materialFunction=&MF_CharacterParent_2023_PreFX}
  MaterialFunctionCall(
   Material=StaticSwitchParameter(
    True=SetMaterialAttributes!(MaterialAttributes=@n2, World Position Offset=@n8) {attributeSetTypes [#1]}
    False=@n2
   ) {group=Flipbook Eyes, parameterName=Debug Eye Poses}
   M=@n9.M
   Pre Skinned Local Position=@n9.M
   Distance Blend=@n9.M
   Fresnel=@n9.M
  ) {materialFunction=&MF_CharacterParent_2023_PostFX}
  Min(A=Min(A=@n4.R, B=@n4.R), B=@n4.B)
  NamedRerouteDeclaration(None=MaterialFunctionCall!(Material=@n9.Result) {materialFunction=/Game/Characters/Player/Common/MaterialFunctions/VertexColorFaces/MF_FV2_IndexEyes}.Result) {name=MA Mouth In, nodeColor=<rgba 0.9922 0 1 1>}
  n10 = Round(None=@n1)
  n11 = NamedRerouteUsage()
  n12 = Divide(A=Constant2Vector() {g=1, r=1}, B=@n11) {desc=Scale}
  n13 = ComponentMask(None=@n12) {desc="1 / ## Horz Images", r=true}
  n14 = AppendVector(
   A=Fmod(A=@n10, B=ComponentMask(None=@n11) {desc="## Horz Images", r=true}) {desc=IndexA_Horz}
   B=Floor(None=Multiply(A=@n10, B=@n13)) {desc=IndexA_Vert}
  ) {desc=IndexA_Combined}
  n15 = NamedRerouteUsage()
  NamedRerouteDeclaration(None=Multiply(A=Add(A=@n14, B=@n15), B=@n12) {desc=UV_A}) {name=Mouth_FlipbookUV, nodeColor=<rgba 1 0 0.4453 1>}
  n16 = TextureCoordinate() {coordinateIndex=2}
  n17 = TextureCoordinate()
  NamedRerouteDeclaration(None=StaticSwitchParameter(
    True=TextureCoordinate() {coordinateIndex=1}
    False=StaticSwitchParameter(True=@n16, False=@n17) {group=09_Flipbook Mouth, parameterName=UseUV3forMouth}
   ) {group=09_Flipbook Mouth, parameterName=UseUV2forMouth}) {name=MouthUVInput, nodeColor=<rgba 0.1016 0 1 1>}
  n18 = NamedRerouteUsage()
  n19 = GetMaterialAttributes(None=NamedRerouteUsage()) {attributeGetTypes [#2, #3]}
  n20 = TextureSampleParameter2D(UVs=NamedRerouteUsage()) {
   group=09_Flipbook Mouth
   parameterName=Mouth1Flipbook
   samplerSource=SSM_Wrap_WorldGroupSettings
   texture=/Game/Characters/Player/Male/Medium/Bodies/M_MED_CarolinaChili/Textures/T_CarolinaChili_Mouth_D
  }
  n21 = Constant() {r=0.5}
  n22 = MaterialFunctionCall(Float2=NamedRerouteUsage()) {materialFunction=&BreakOutFloat2Components}
  n23 = Multiply(
   A=ComponentMask(None=@n20.RGBA) {a=true}
   B=Multiply(A=SphereMask(A=@n21, B=@n22.R) {attenuationRadius=0.5}, B=SphereMask(A=@n22.G, B=@n21) {attenuationRadius=0.5})
  )
  n24 = ComponentMask(None=@n20.RGBA) {b=true, g=true, r=true}
  n25 = SetMaterialAttributes!(
   MaterialAttributes=@n19.MaterialAttributes
   Base Color=LinearInterpolate(A=@n19.Base Color, B=@n24, Alpha=@n23)
   Emissive Color=@n19.Emissive Color
  ) {attributeSetTypes [#2, #3]}
  n26 = Multiply(A=@n24, B=ScalarParameter() {defaultValue=0.3, group=09_Flipbook Mouth, parameterName=MouthEmissiveIntensity})
  n27 = SetMaterialAttributes!(
   MaterialAttributes=@n25
   Shading Model=ShadingModel()
   Normal=Constant3Vector() {constant=<rgba 0 0 1 0>}
   Roughness=Constant() {r=1}
   Specular=Constant()
   Emissive Color=StaticSwitchParameter(True=LinearInterpolate(A=@n19.Emissive Color, B=@n26, Alpha=@n23) {constB=0}, False=@n19.Emissive Color) {group=09_Flipbook Mouth, parameterName=useEmissiveforMouth}
  ) {attributeSetTypes [
    #4
    #5
    #6
    #7
    #3
   ]}
  n28 = TextureSampleParameter2D(UVs=NamedRerouteUsage()) {
   group=09_Flipbook Mouth
   parameterName=Mouth FX Map
   samplerSource=SSM_Wrap_WorldGroupSettings
   texture=/Game/Characters/Player/Male/Medium/Bodies/M_MED_CarolinaChili/Textures/T_CarolinaChili_Mouth_FX
  }
  n29 = BlendMaterialAttributes(A=@n27, B=@n25, Alpha=@n28.G) {vertexAttributeBlendType=UseA}
  NamedRerouteDeclaration(None=StaticSwitchParameter(True=BlendMaterialAttributes(A=@n18, B=@n29, Alpha=@n23), False=@n18) {group=09_Flipbook Mouth, parameterName=Enable Mouth Flipbook, sortPriority=-32}) {name=MA Mouth Out, nodeColor=<rgba 1 0 0.5391 1>}
  NamedRerouteDeclaration(None=NamedRerouteUsage()) {name=Mouth_Frame, nodeColor=<rgba 1 0.5156 0 1>}
  NamedRerouteDeclaration(None=AppendVector(
    A=ScalarParameter() {defaultValue=4, group=09_Flipbook Mouth, parameterName=Mouth_SubimagesX}
    B=ScalarParameter() {defaultValue=4, group=09_Flipbook Mouth, parameterName=Mouth_SubimagesY}
   )) {name=Mouth_Subimages, nodeColor=<rgba 0.1953 0 1 1>}
  n30 = ScalarParameter() {defaultValue=0.305, group=09_Flipbook Mouth, parameterName=Mouth1UVShiftX}
  n31 = ScalarParameter() {defaultValue=0.15, group=09_Flipbook Mouth, parameterName=Mouth1UVShiftY}
  n32 = Subtract(A=Subtract(A=NamedRerouteUsage(), B=AppendVector(A=@n30, B=@n31))) {constB=0.5}
  n33 = ScalarParameter() {defaultValue=15, group=09_Flipbook Mouth, parameterName=Mouth1Scale}
  NamedRerouteDeclaration(None=Add(A=Multiply(A=@n32, B=@n33)) {constB=0.5}) {name=Mouth_UV, nodeColor=<rgba 0.125 1 0 1>}
  n34 = Time()
  n35 = ScalarParameter() {parameterName=flipbook_face_mouth_index}
  n36 = ScalarParameter() {group=DEBUG, parameterName=DEBUG mouth index offset}
  NamedRerouteDeclaration(None=StaticSwitchParameter(True=Floor(None=@n34), False=Add(A=@n35, B=@n36)) {parameterName=DEBUG Cycle Mouth Index}) {name=Mouth_Index, nodeColor=<rgba 0.0312 1 0 1>}
 ]
}
