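# Physics asset for the CarolinaChili medium male player body, built for secondary motion on collar, jacket, belts, grenades and shoe slaps.
# Capsules cover the main body, while tiny deform_ capsules anchor belt, grenade and shoe pieces.
# Dyn_ bodies have tight angular limits to spine or deform_ anchors, with looser links between some neighbours.
# Body joints are angularly free, so they lack ragdoll limits; only the left centre shoe-slap anchor is locked to its thigh.
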PhysicsAsset /Game/Characters/Player/Male/Medium/Bodies/M_MED_CarolinaChili/Meshes/M_MED_CarolinaChili_Physics
bodies {
 clavicle_l [{shapeType=Capsule, length=5.3376, radius=9.3813, shapeName=clavicle_l_capsule, center=<10.3018 -1.0348 0>, rotation=<rot 90 0 0>}]
 clavicle_r [{shapeType=Capsule, length=5.3376, radius=9.3813, shapeName=clavicle_l_capsule, center=<-10.3018 1.0348 0>, rotation=<rot 90 0 180>}]
 deform_grenade_b_l [{shapeType=Capsule, length=0.505, radius=0.505, shapeName=deform_grenade_b_l_capsule, rotation=<rot 0 -90 0>}]
 deform_grenade_b_r [{shapeType=Capsule, length=0.505, radius=0.505, shapeName=deform_grenade_b_r_capsule, rotation=<rot 0 -90 0>}]
 deform_grenade_f_l [{shapeType=Capsule, length=0.505, radius=0.505, shapeName=deform_grenade_f_l_capsule, rotation=<rot 0 -90 0>}]
 deform_grenade_f_r [{shapeType=Capsule, length=0.505, radius=0.505, shapeName=deform_grenade_f_r_capsule, rotation=<rot 0 -90 0>}]
 deform_jacket_belt_l [{shapeType=Capsule, length=0.505, radius=0.505, shapeName=deform_jacket_belt_l_capsule, rotation=<rot 0 -90 0>}]
 deform_jacket_bert_r [{shapeType=Capsule, length=0.505, radius=0.505, shapeName=deform_jacket_bert_r_capsule, rotation=<rot 0 -90 0>}]
 deform_shoe_slap_c_l [{shapeType=Capsule, length=0.5, radius=0.5, shapeName=deform_shoe_slap_c_l_capsule, rotation=<rot 0 -90 0>}]
 deform_shoe_slap_c_r [{shapeType=Capsule, length=0.505, radius=0.505, shapeName=deform_shoe_slap_c_r_capsule, rotation=<rot 0 -90 0>}]
 deform_shoe_slap_f_l [{shapeType=Capsule, length=0.505, radius=0.505, shapeName=deform_shoe_slap_f_l_capsule, rotation=<rot 0 -90 0>}]
 deform_shoe_slap_f_r [{shapeType=Capsule, length=0.505, radius=0.505, shapeName=deform_shoe_slap_f_r_capsule, rotation=<rot 0 -90 0>}]
 dyn_collar_b [{shapeType=Capsule, length=21.447, radius=2.1345, shapeName=dyn_collar_b_capsule, center=<4.3448 0.7768 0>}]
 dyn_collar_f_l [{shapeType=Capsule, length=13.5071, radius=2.2488, shapeName=dyn_collar_f_l_capsule, center=<3.2702 2.4716 0>, rotation=<rot 0 0 -180>}]
 dyn_collar_f_r [{shapeType=Capsule, length=13.5071, radius=2.2488, shapeName=dyn_collar_f_l_capsule, center=<-3.2702 -2.4716 0>}]
 dyn_collar_m_l [{shapeType=Capsule, length=6.6462, radius=3.0911, shapeName=dyn_collar_m_l_capsule, center=<6.932 1.5336 0>, rotation=<rot 90 0 0>}]
 dyn_collar_m_r [{shapeType=Capsule, length=4.7543, radius=3.0911, shapeName=dyn_collar_m_l_capsule, center=<-9.1574 -2.3387 0>, rotation=<rot 90 0 0>}]
 dyn_grenade_b_l [{
   shapeType=Capsule
   length=10.8508
   radius=3.0285
   shapeName=dyn_grenade_b_l_capsule
   center=<2.8392 0 -2.9907>
   rotation=<rot 3.389 0 -0.0003>
  }]
 dyn_grenade_b_r [{
   shapeType=Capsule
   length=10.0718
   radius=3.8075
   shapeName=dyn_grenade_b_l_capsule
   center=<-3.8155 0 2.4412>
   rotation=<rot 3.389 0 179.9997>
  }]
 dyn_grenade_f_l [{
   shapeType=Capsule
   length=10.0718
   radius=3.8075
   shapeName=dyn_grenade_f_l_capsule
   center=<2.8392 0 -2.9907>
   rotation=<rot 3.3892 0 -0.0003>
  }]
 dyn_grenade_f_r [{
   shapeType=Capsule
   length=10.0718
   radius=3.8075
   shapeName=dyn_grenade_f_l_capsule
   center=<-3.6521 0 2.9426>
   rotation=<rot 3.3892 0 179.9997>
  }]
 dyn_jacket_b [{
   shapeType=Capsule
   length=16.9325
   radius=5.0528
   shapeName=dyn_jacket_b_capsule
   center=<7.503 4.4104 -0.0159>
   rotation=<rot -3.078 -90 1.9487>
  }]
 dyn_jacket_b_l [{shapeType=Capsule, length=8.5347, radius=6.9009, shapeName=dyn_jacket_b_l_capsule, center=<7.3849 3.9975 0>}]
 dyn_jacket_b_r [{shapeType=Capsule, length=8.5347, radius=6.9009, shapeName=dyn_jacket_b_l_capsule, center=<-7.3849 -3.9975 0>, rotation=<rot 0 0 -180>}]
 dyn_jacket_belt_l [{shapeType=Capsule, length=3.2295, radius=2.3468, shapeName=dyn_jacket_belt_l_capsule, center=<3.6969 0 0>, rotation=<rot 90 0 0>}]
 dyn_jacket_bert_r [{shapeType=Capsule, length=5.1629, radius=2.3258, shapeName=dyn_jacket_bert_r_capsule, center=<-2.5273 0 0>, rotation=<rot 90 0 0>}]
 dyn_jacket_f_l [{shapeType=Capsule, length=10.9294, radius=2.4938, shapeName=dyn_jacket_f_l_capsule, center=<8.9963 1.1193 -2.8302>}]
 dyn_jacket_f_r [{shapeType=Capsule, length=10.9294, radius=2.4938, shapeName=dyn_jacket_f_l_capsule, center=<-9.8669 -0.6814 2.8302>, rotation=<rot 0 0 -180>}]
 dyn_shoe_slap_c_l [{shapeType=Capsule, length=2.4388, radius=5, shapeName=dyn_shoe_slap_c_l_capsule, center=<-0.5336 0 0>, rotation=<rot 90 0 0>}]
 dyn_shoe_slap_c_r [{shapeType=Capsule, length=2.4388, radius=5, shapeName=dyn_shoe_slap_c_l_capsule, center=<0.5336 0 0>, rotation=<rot 90 0 180>}]
 dyn_shoe_slap_f_l [{shapeType=Capsule, length=3.3959, radius=3.2645, shapeName=dyn_shoe_slap_f_l_capsule, center=<4.0744 0 1.8293>, rotation=<rot 90 0 0>}]
 dyn_shoe_slap_f_r [{shapeType=Capsule, length=3.3959, radius=3.2645, shapeName=dyn_shoe_slap_f_l_capsule, center=<-4.0744 0 -1.8293>, rotation=<rot 90 0 180>}]
 foot_l [
  {shapeType=Capsule, length=7.859, radius=7.2235, shapeName=foot_l_capsule, center=<-1.791 3.1288 0>, rotation=<rot 90 -26.5651 14.0818>}
  {shapeType=Capsule, length=9.1757, radius=5.8243, shapeName=foot_r_capsule, center=<-5.6914 10.5191 0>, rotation=<rot -0.0001 7.3307 -90>}
 ]
 foot_r [
  {shapeType=Capsule, length=7.859, radius=7.2235, shapeName=foot_l_capsule, center=<1.791 -3.1288 0>, rotation=<rot 90 -26.5651 -165.9182>}
  {shapeType=Capsule, length=9.1757, radius=5.8243, shapeName=foot_r_capsule, center=<5.6914 -10.5191 0>, rotation=<rot -0.0001 7.3307 90>}
 ]
 head [{shapeType=Capsule, length=8.1638, radius=7.4761, shapeName=head_capsule, center=<4.8537 0 0>, rotation=<rot 90 0 0>}]
 lowerarm_l [{shapeType=Capsule, length=16.1568, radius=5.6895, shapeName=lowerarm_l_capsule, center=<5.9486 0 4.3496>, rotation=<rot 90 0 0>}]
 lowerarm_r [{shapeType=Capsule, length=16.1568, radius=5.6895, shapeName=lowerarm_l_capsule, center=<-5.9486 0 -4.3496>, rotation=<rot 90 0 180>}]
 neck_01 [{shapeType=Capsule, length=2.8349, radius=7.1651, shapeName=neck_01_capsule, rotation=<rot 90 0 0>}]
 neck_02 [{shapeType=Capsule, length=2.8349, radius=7.1651, shapeName=neck_02_capsule, rotation=<rot 90 0 0>}]
 pelvis [
  {shapeType=Capsule, length=1.8914, radius=18.0872, shapeName=pelvis_capsule, center=<-2.7034 -3.5543 0.0338>, rotation=<rot 0 -0.0002 0>}
  {shapeType=Capsule, length=5.3642, radius=15.4565, shapeName=pelvis_capsule, center=<-4.8353 5.0787 0.3827>, rotation=<rot -3.2548 -0.0007 0>}
 ]
 spine_01 [{shapeType=Capsule, length=7.359, radius=11.7261, shapeName=spine_01_capsule, center=<6.8774 0.7601 0>, rotation=<rot 90 0 0.0001>}]
 spine_02 [{shapeType=Capsule, length=5.3393, radius=11.6607, shapeName=spine_02_capsule, center=<10 0.7449 0>}]
 spine_03 [{shapeType=Capsule, length=5.1471, radius=11.8529, shapeName=spine_03_capsule, center=<10 2.5392 0>}]
 spine_04 [{shapeType=Capsule, length=5.2524, radius=13.3319, shapeName=spine_04_capsule, center=<10 0 0>}]
 spine_05 [{shapeType=Capsule, length=5, radius=12, shapeName=spine_05_capsule, center=<4.4941 0 0>}]
 thigh_l [{shapeType=Capsule, length=20, radius=10, shapeName=thigh_l_capsule, center=<-15.6948 0 4.0326>, rotation=<rot 90 0 0>}]
 thigh_r [{shapeType=Capsule, length=20, radius=10, shapeName=thigh_l_capsule, center=<15.6948 0 -4.0326>, rotation=<rot 90 0 180>}]
 upperarm_l [{shapeType=Capsule, length=17.3066, radius=6.4028, shapeName=upperarm_l_capsule, center=<8.9727 0 4.1775>, rotation=<rot 90 0 0>}]
 upperarm_r [{shapeType=Capsule, length=17.3066, radius=6.4028, shapeName=upperarm_l_capsule, center=<-8.9727 0 -4.1775>, rotation=<rot 90 0 180>}]
}
constraints [
 {
  bone1Name=neck_02
  bone2Name=neck_01
  swing1LimitDegrees=45
  swing1Motion=Free
  swing2LimitDegrees=45
  swing2Motion=Free
  twistLimitDegrees=45
  twistMotion=Free
 }
 {
  bone1Name=head
  bone2Name=neck_02
  swing1LimitDegrees=45
  swing1Motion=Free
  swing2LimitDegrees=45
  swing2Motion=Free
  twistLimitDegrees=45
  twistMotion=Free
 }
 {
  bone1Name=neck_01
  bone2Name=spine_05
  swing1LimitDegrees=45
  swing1Motion=Free
  swing2LimitDegrees=45
  swing2Motion=Free
  twistLimitDegrees=45
  twistMotion=Free
 }
 {bone1Name=dyn_collar_b, bone2Name=spine_05, swing1LimitDegrees=45, swing2LimitDegrees=10, twistLimitDegrees=10}
 {bone1Name=dyn_jacket_f_l, bone2Name=spine_03, swing1LimitDegrees=25, swing2LimitDegrees=1, twistLimitDegrees=5}
 {bone1Name=dyn_jacket_b_l, bone2Name=spine_03, swing1LimitDegrees=30, swing2LimitDegrees=10, twistLimitDegrees=10}
 {bone1Name=dyn_jacket_belt_l, bone2Name=deform_jacket_belt_l, swing1LimitDegrees=30, swing2LimitDegrees=3, twistLimitDegrees=3}
 {bone1Name=dyn_jacket_bert_r, bone2Name=deform_jacket_bert_r, swing1LimitDegrees=30, swing2LimitDegrees=3, twistLimitDegrees=3}
 {bone1Name=dyn_grenade_f_l, bone2Name=deform_grenade_f_l, swing1LimitDegrees=5, swing2LimitDegrees=5, twistLimitDegrees=5}
 {bone1Name=dyn_grenade_b_l, bone2Name=deform_grenade_b_l, swing1LimitDegrees=5, swing2LimitDegrees=5, twistLimitDegrees=5}
 {
  bone1Name=spine_02
  bone2Name=spine_01
  swing1LimitDegrees=45
  swing1Motion=Free
  swing2LimitDegrees=45
  swing2Motion=Free
  twistLimitDegrees=45
  twistMotion=Free
 }
 {
  bone1Name=spine_03
  bone2Name=spine_02
  swing1LimitDegrees=45
  swing1Motion=Free
  swing2LimitDegrees=45
  swing2Motion=Free
  twistLimitDegrees=45
  twistMotion=Free
 }
 {
  bone1Name=spine_05
  bone2Name=spine_04
  swing1LimitDegrees=45
  swing1Motion=Free
  swing2LimitDegrees=45
  swing2Motion=Free
  twistLimitDegrees=45
  twistMotion=Free
 }
 {
  bone1Name=spine_04
  bone2Name=spine_03
  swing1LimitDegrees=45
  swing1Motion=Free
  swing2LimitDegrees=45
  swing2Motion=Free
  twistLimitDegrees=45
  twistMotion=Free
 }
 {
  bone1Name=clavicle_l
  bone2Name=spine_05
  swing1LimitDegrees=45
  swing1Motion=Free
  swing2LimitDegrees=45
  swing2Motion=Free
  twistLimitDegrees=45
  twistMotion=Free
 }
 {
  bone1Name=upperarm_l
  bone2Name=clavicle_l
  swing1LimitDegrees=45
  swing1Motion=Free
  swing2LimitDegrees=45
  swing2Motion=Free
  twistLimitDegrees=45
  twistMotion=Free
 }
 {
  bone1Name=lowerarm_l
  bone2Name=upperarm_l
  swing1LimitDegrees=45
  swing1Motion=Free
  swing2LimitDegrees=45
  swing2Motion=Free
  twistLimitDegrees=45
  twistMotion=Free
 }
 {
  bone1Name=spine_01
  bone2Name=pelvis
  swing1LimitDegrees=45
  swing1Motion=Free
  swing2LimitDegrees=45
  swing2Motion=Free
  twistLimitDegrees=45
  twistMotion=Free
 }
 {bone1Name=dyn_shoe_slap_c_l, bone2Name=deform_shoe_slap_c_l, swing1LimitDegrees=15, swing2LimitDegrees=15, twistLimitDegrees=5}
 {bone1Name=dyn_shoe_slap_f_l, bone2Name=deform_shoe_slap_f_l, swing1LimitDegrees=10, swing2LimitDegrees=30, twistLimitDegrees=5}
 {
  bone1Name=clavicle_r
  bone2Name=spine_05
  swing1LimitDegrees=45
  swing1Motion=Free
  swing2LimitDegrees=45
  swing2Motion=Free
  twistLimitDegrees=45
  twistMotion=Free
 }
 {bone1Name=dyn_collar_m_r, bone2Name=spine_05, swing1LimitDegrees=60, swing2LimitDegrees=10, twistLimitDegrees=1}
 {bone1Name=dyn_collar_f_r, bone2Name=spine_05, swing1LimitDegrees=60, swing2LimitDegrees=10, twistLimitDegrees=1}
 {bone1Name=dyn_jacket_b_r, bone2Name=spine_03, swing1LimitDegrees=30, swing2LimitDegrees=10, twistLimitDegrees=10}
 {bone1Name=dyn_grenade_f_r, bone2Name=deform_grenade_f_r, swing1LimitDegrees=5, swing2LimitDegrees=5, twistLimitDegrees=5}
 {bone1Name=dyn_grenade_b_r, bone2Name=deform_grenade_b_r, swing1LimitDegrees=5, swing2LimitDegrees=5, twistLimitDegrees=5}
 {bone1Name=dyn_shoe_slap_f_r, bone2Name=deform_shoe_slap_f_r, swing1LimitDegrees=10, swing2LimitDegrees=30, twistLimitDegrees=5}
 {bone1Name=dyn_shoe_slap_c_r, bone2Name=deform_shoe_slap_c_r, swing1LimitDegrees=15, swing2LimitDegrees=15, twistLimitDegrees=5}
 {
  bone1Name=foot_r
  bone2Name=thigh_r
  swing1LimitDegrees=45
  swing1Motion=Free
  swing2LimitDegrees=45
  swing2Motion=Free
  twistLimitDegrees=45
  twistMotion=Free
 }
 {
  bone1Name=deform_shoe_slap_f_r
  bone2Name=deform_shoe_slap_c_r
  swing1LimitDegrees=45
  swing1Motion=Free
  swing2LimitDegrees=45
  swing2Motion=Free
  twistLimitDegrees=45
  twistMotion=Free
 }
 {
  bone1Name=deform_shoe_slap_f_l
  bone2Name=deform_shoe_slap_c_l
  swing1LimitDegrees=45
  swing1Motion=Free
  swing2LimitDegrees=45
  swing2Motion=Free
  twistLimitDegrees=45
  twistMotion=Free
 }
 {
  bone1Name=deform_shoe_slap_c_r
  bone2Name=thigh_r
  swing1LimitDegrees=45
  swing1Motion=Free
  swing2LimitDegrees=45
  swing2Motion=Free
  twistLimitDegrees=45
  twistMotion=Free
 }
 {
  bone1Name=deform_shoe_slap_c_l
  bone2Name=thigh_l
  swing1LimitDegrees=45
  swing1Motion=Locked
  swing2LimitDegrees=45
  swing2Motion=Locked
  twistLimitDegrees=45
  twistMotion=Locked
 }
 {
  bone1Name=dyn_collar_f_r
  bone2Name=dyn_collar_m_r
  swing1LimitDegrees=10
  swing2LimitDegrees=10
  swing2Motion=Free
  twistLimitDegrees=10
  twistMotion=Free
 }
 {bone1Name=dyn_collar_m_l, bone2Name=spine_05, swing1LimitDegrees=60, swing2LimitDegrees=10, twistLimitDegrees=1}
 {bone1Name=dyn_collar_f_l, bone2Name=spine_05, swing1LimitDegrees=60, swing2LimitDegrees=10, twistLimitDegrees=1}
 {
  bone1Name=dyn_collar_f_l
  bone2Name=dyn_collar_m_l
  swing1LimitDegrees=10
  swing2LimitDegrees=45
  swing2Motion=Free
  twistLimitDegrees=45
  twistMotion=Free
 }
 {bone1Name=dyn_jacket_b, bone2Name=spine_03, swing1LimitDegrees=30, swing2LimitDegrees=1, twistLimitDegrees=5}
 {
  bone1Name=dyn_shoe_slap_f_l
  bone2Name=dyn_shoe_slap_c_l
  swing1LimitDegrees=45
  swing1Motion=Free
  swing2LimitDegrees=45
  swing2Motion=Free
  twistLimitDegrees=45
  twistMotion=Free
 }
 {
  bone1Name=dyn_shoe_slap_f_r
  bone2Name=dyn_shoe_slap_c_r
  swing1LimitDegrees=45
  swing1Motion=Free
  swing2LimitDegrees=45
  swing2Motion=Free
  twistLimitDegrees=45
  twistMotion=Free
 }
 {
  bone1Name=thigh_l
  bone2Name=pelvis
  swing1LimitDegrees=45
  swing1Motion=Free
  swing2LimitDegrees=45
  swing2Motion=Free
  twistLimitDegrees=45
  twistMotion=Free
 }
 {
  bone1Name=thigh_r
  bone2Name=pelvis
  swing1LimitDegrees=45
  swing1Motion=Free
  swing2LimitDegrees=45
  swing2Motion=Free
  twistLimitDegrees=45
  twistMotion=Free
 }
 {
  bone1Name=dyn_grenade_b_l
  bone2Name=dyn_grenade_f_l
  swing1LimitDegrees=45
  swing1Motion=Free
  swing2LimitDegrees=45
  swing2Motion=Free
  twistLimitDegrees=45
  twistMotion=Free
 }
 {
  bone1Name=dyn_grenade_b_r
  bone2Name=dyn_grenade_f_r
  swing1LimitDegrees=45
  swing1Motion=Free
  swing2LimitDegrees=45
  swing2Motion=Free
  twistLimitDegrees=45
  twistMotion=Free
 }
 {
  bone1Name=deform_grenade_b_l
  bone2Name=spine_01
  swing1LimitDegrees=45
  swing1Motion=Free
  swing2LimitDegrees=45
  swing2Motion=Free
  twistLimitDegrees=45
  twistMotion=Free
 }
 {
  bone1Name=deform_grenade_b_r
  bone2Name=spine_01
  swing1LimitDegrees=45
  swing1Motion=Free
  swing2LimitDegrees=45
  swing2Motion=Free
  twistLimitDegrees=45
  twistMotion=Free
 }
 {
  bone1Name=deform_grenade_f_l
  bone2Name=spine_01
  swing1LimitDegrees=45
  swing1Motion=Free
  swing2LimitDegrees=45
  swing2Motion=Free
  twistLimitDegrees=45
  twistMotion=Free
 }
 {
  bone1Name=deform_grenade_f_r
  bone2Name=spine_01
  swing1LimitDegrees=45
  swing1Motion=Free
  swing2LimitDegrees=45
  swing2Motion=Free
  twistLimitDegrees=45
  twistMotion=Free
 }
 {bone1Name=dyn_jacket_f_r, bone2Name=spine_03, swing1LimitDegrees=30, swing2LimitDegrees=10, twistLimitDegrees=20}
 {
  bone1Name=foot_l
  bone2Name=thigh_l
  swing1LimitDegrees=45
  swing1Motion=Free
  swing2LimitDegrees=45
  swing2Motion=Free
  twistLimitDegrees=45
  twistMotion=Free
 }
 {
  bone1Name=upperarm_r
  bone2Name=clavicle_r
  swing1LimitDegrees=45
  swing1Motion=Free
  swing2LimitDegrees=45
  swing2Motion=Free
  twistLimitDegrees=45
  twistMotion=Free
 }
 {
  bone1Name=lowerarm_r
  bone2Name=upperarm_r
  swing1LimitDegrees=45
  swing1Motion=Free
  swing2LimitDegrees=45
  swing2Motion=Free
  twistLimitDegrees=45
  twistMotion=Free
 }
 {
  bone1Name=deform_jacket_belt_l
  bone2Name=spine_03
  swing1LimitDegrees=45
  swing1Motion=Free
  swing2LimitDegrees=45
  swing2Motion=Free
  twistLimitDegrees=45
  twistMotion=Free
 }
 {
  bone1Name=deform_jacket_bert_r
  bone2Name=spine_03
  swing1LimitDegrees=45
  swing1Motion=Free
  swing2LimitDegrees=45
  swing2Motion=Free
  twistLimitDegrees=45
  twistMotion=Free
 }
]
