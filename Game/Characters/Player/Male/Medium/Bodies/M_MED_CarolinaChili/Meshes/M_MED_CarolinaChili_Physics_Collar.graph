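# Physics asset for the collar of the CarolinaChili medium male player body, covering five collar pieces: back, front left/right and middle left/right.
# Head, neck and upper-spine capsules form the surrounding body shapes, chained with free angular constraints.
# Each dyn_collar capsule hangs off a tiny deform_collar anchor with tight swing and twist limits, looser on the middle pieces.
# The anchors attach to spine_05 with free angular motion.
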PhysicsAsset /Game/Characters/Player/Male/Medium/Bodies/M_MED_CarolinaChili/Meshes/M_MED_CarolinaChili_Physics_Collar
bodies {
 deform_collar_b [{shapeType=Capsule, length=0.505, radius=0.505, shapeName=deform_collar_b_capsule, rotation=<rot 0 -90 0>}]
 deform_collar_f_l [{shapeType=Capsule, length=0.505, radius=0.505, shapeName=deform_collar_f_l_capsule, rotation=<rot 0 -90 0>}]
 deform_collar_f_r [{shapeType=Capsule, length=0.505, radius=0.505, shapeName=deform_collar_f_r_capsule, rotation=<rot 0 -90 0>}]
 deform_collar_m_l [{shapeType=Capsule, length=0.505, radius=0.505, shapeName=deform_collar_m_l_capsule, rotation=<rot 0 -90 0>}]
 deform_collar_m_r [{shapeType=Capsule, length=0.505, radius=0.505, shapeName=deform_collar_m_r_capsule, rotation=<rot 0 -90 0>}]
 dyn_collar_b [{shapeType=Capsule, length=21.447, radius=3, shapeName=dyn_collar_b_capsule, center=<1.6073 0 0>}]
 dyn_collar_f_l [{shapeType=Capsule, length=13.5071, radius=3, shapeName=dyn_collar_f_l_capsule, center=<2 0 0>, rotation=<rot 0 0 -180>}]
 dyn_collar_f_r [{shapeType=Capsule, length=13.5071, radius=3, shapeName=dyn_collar_f_l_capsule, center=<-2 0 0>}]
 dyn_collar_m_l [{shapeType=Capsule, length=9.1196, radius=3, shapeName=dyn_collar_m_l_capsule, center=<6.4571 -0.0474 -1.1765>, rotation=<rot 90 0 0>}]
 dyn_collar_m_r [{shapeType=Capsule, length=8.1961, radius=2.7654, shapeName=dyn_collar_m_l_capsule, center=<-6.4571 0.0474 1.1765>, rotation=<rot 90 180 0>}]
 head [{shapeType=Capsule, length=8.1638, radius=7.4761, shapeName=head_capsule, center=<4.8537 0 0>, rotation=<rot 90 0 0>}]
 neck_01 [{shapeType=Capsule, length=2.8349, radius=7.1651, shapeName=neck_01_capsule, rotation=<rot 90 0 0>}]
 neck_02 [{shapeType=Capsule, length=2.8349, radius=7.1651, shapeName=neck_02_capsule, rotation=<rot 90 0 0>}]
 spine_04 [{shapeType=Capsule, length=5.2524, radius=13.3319, shapeName=spine_04_capsule, center=<10 0 0>}]
 spine_05 [{shapeType=Capsule, length=4.2394, radius=12, shapeName=spine_05_capsule, center=<4.4941 -1.4648 0>}]
}
constraints [
 {
  bone1Name=neck_02
  bone2Name=neck_01
  swing1LimitDegrees=45
  swing1Motion=Free
  swing2LimitDegrees=45
  swing2Motion=Free
  twistLimitDegrees=45
  twistMotion=Free
 }
 {
  bone1Name=head
  bone2Name=neck_02
  swing1LimitDegrees=45
  swing1Motion=Free
  swing2LimitDegrees=45
  swing2Motion=Free
  twistLimitDegrees=45
  twistMotion=Free
 }
 {
  bone1Name=neck_01
  bone2Name=spine_05
  swing1LimitDegrees=45
  swing1Motion=Free
  swing2LimitDegrees=45
  swing2Motion=Free
  twistLimitDegrees=45
  twistMotion=Free
 }
 {bone1Name=dyn_collar_b, bone2Name=deform_collar_b, swing1LimitDegrees=5, swing2LimitDegrees=5, twistLimitDegrees=1}
 {
  bone1Name=spine_05
  bone2Name=spine_04
  swing1LimitDegrees=45
  swing1Motion=Free
  swing2LimitDegrees=45
  swing2Motion=Free
  twistLimitDegrees=45
  twistMotion=Free
 }
 {bone1Name=dyn_collar_f_r, bone2Name=deform_collar_f_r, swing1LimitDegrees=5, swing2LimitDegrees=5, twistLimitDegrees=1}
 {bone1Name=dyn_collar_f_l, bone2Name=deform_collar_f_l, swing1LimitDegrees=5, swing2LimitDegrees=5, twistLimitDegrees=1}
 {
  bone1Name=deform_collar_m_r
  bone2Name=spine_05
  swing1LimitDegrees=45
  swing1Motion=Free
  swing2LimitDegrees=45
  swing2Motion=Free
  twistLimitDegrees=45
  twistMotion=Free
 }
 {
  bone1Name=deform_collar_m_l
  bone2Name=spine_05
  swing1LimitDegrees=45
  swing1Motion=Free
  swing2LimitDegrees=45
  swing2Motion=Free
  twistLimitDegrees=45
  twistMotion=Free
 }
 {
  bone1Name=deform_collar_f_r
  bone2Name=spine_05
  swing1LimitDegrees=45
  swing1Motion=Free
  swing2LimitDegrees=45
  swing2Motion=Free
  twistLimitDegrees=45
  twistMotion=Free
 }
 {
  bone1Name=deform_collar_f_l
  bone2Name=spine_05
  swing1LimitDegrees=45
  swing1Motion=Free
  swing2LimitDegrees=45
  swing2Motion=Free
  twistLimitDegrees=45
  twistMotion=Free
 }
 {
  bone1Name=deform_collar_b
  bone2Name=spine_05
  swing1LimitDegrees=45
  swing1Motion=Free
  swing2LimitDegrees=45
  swing2Motion=Free
  twistLimitDegrees=45
  twistMotion=Free
 }
 {bone1Name=dyn_collar_m_l, bone2Name=deform_collar_m_l, swing1LimitDegrees=15, swing2LimitDegrees=5, twistLimitDegrees=1}
 {bone1Name=dyn_collar_m_r, bone2Name=deform_collar_m_r, swing1LimitDegrees=20, swing2LimitDegrees=10, twistLimitDegrees=1}
]
