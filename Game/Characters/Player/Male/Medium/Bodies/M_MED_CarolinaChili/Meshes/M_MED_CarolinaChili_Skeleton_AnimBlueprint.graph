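# Post-process style AnimBlueprint for the CarolinaChili male body part, adding secondary motion to jacket, collar, zippers and shoes.
# Copied parent pose runs through two Control Rigs, two RigidBody sims, weighted constraints and AnimDynamics.
# Cached poses then layer backpack jacket correction, skydiving collar scale, partial head counter-rotation outside emotes, front-end jacket tweaks and belt constraints.
# EventGraph updates gravity override, head rotation, LOD and emote flags, detecting backbling on customization load.
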AnimBlueprint /Game/Characters/Player/Male/Medium/Bodies/M_MED_CarolinaChili/Meshes/M_MED_CarolinaChili_Skeleton_AnimBlueprint
target_skeleton=&/M_MED_CarolinaChili_Skeleton
graphs [
 AnimGraph {nodes [
   n1 = "Constraint - dyn_jacket_f_l (R)"(ComponentPose=Local To Component(LocalPose=Use cached pose 'prefrontend'()), Alpha=0.7, ConstraintWeights_0=1) {
    alphaBoolBlend {blendOption=Linear}
    boneToModify {boneName=dyn_jacket_f_l}
    constraintSetup [{transformType=Rotation, targetBone {boneName=deform_jacket_f_l}}]
    constraintWeights [1]
   }
   n2 = "Constraint - dyn_jacket_f_r (R)"(ComponentPose=@n1, Alpha=0.7, ConstraintWeights_0=1) {
    alphaBoolBlend {blendOption=Linear}
    boneToModify {boneName=dyn_jacket_f_r}
    constraintSetup [{transformType=Rotation, targetBone {boneName=deform_jacket_f_r}}]
    constraintWeights [1]
   }
   n3 = "Transform (Modify) Bone - Bone: deform_collar_f_l"(ComponentPose=@n2, Alpha=1, Translation="0,0,0", Rotation="0.000,0.000,15.000", Scale="1,1,1") {
    rotationMode=BMM_Additive
    rotationSpace=BCS_BoneSpace
    translationMode=BMM_Additive
    translationSpace=BCS_BoneSpace
    alphaBoolBlend {blendOption=Linear}
    boneToModify {boneName=deform_collar_f_l}
    rotation=<rot 0.0001 0 15>
   }
   n4 = Component To Local(ComponentPose=@n3)
   n5 = Use cached pose 'prefrontend'()
   n6 = Get bIsInFrontEnd()
   n7 = "Constraint - deform_jacket_belt_l (PPPPPPPP)"!(
    ComponentPose=Local To Component(LocalPose=Blend Poses by bool(BlendPose_0=@n4, BlendPose_1=@n5, bActiveValue=@n6) {blendType=Linear})
    Alpha=1
    ConstraintWeights_0=0.0449
    ConstraintWeights_1:float=0.096
    ConstraintWeights_2:float=0.1463
    ConstraintWeights_3:float=0.1127
    ConstraintWeights_4:float=0.0021
    ConstraintWeights_5:float=0.0004
    ConstraintWeights_6:float=0.488
    ConstraintWeights_7:float=0.1097
   ) -> (Pose) {
    alphaBoolBlend {blendOption=Linear}
    boneToModify {boneName=deform_jacket_belt_l}
    constraintSetup [
     {targetBone {boneName=spine_01}}
     {targetBone {boneName=spine_02}}
     {targetBone {boneName=spine_03}}
     {targetBone {boneName=spine_04}}
     {targetBone {boneName=spine_05}}
     {targetBone {boneName=deform_upperarm_bk_l}}
     {targetBone {boneName=dyn_jacket_f_l}}
     {targetBone {boneName=dyn_jacket_b}}
    ]
    constraintWeights [1,1,1,1,1,1,1,1]
   }
   n8 = "Constraint - deform_jacket_bert_r (PPPPPPPP)"!(
    ComponentPose=@n7
    Alpha=1
    ConstraintWeights_0=0.0248
    ConstraintWeights_1:float=0.0734
    ConstraintWeights_2:float=0.1061
    ConstraintWeights_3:float=0.0651
    ConstraintWeights_4:float=0.0006
    ConstraintWeights_5:float=0.0002
    ConstraintWeights_6:float=0.6889
    ConstraintWeights_7:float=0.041
   ) -> (Pose) {
    alphaBoolBlend {blendOption=Linear}
    boneToModify {boneName=deform_jacket_bert_r}
    constraintSetup [
     {targetBone {boneName=spine_01}}
     {targetBone {boneName=spine_02}}
     {targetBone {boneName=spine_03}}
     {targetBone {boneName=spine_04}}
     {targetBone {boneName=spine_05}}
     {targetBone {boneName=clavicle_r}}
     {targetBone {boneName=dyn_jacket_f_r}}
     {targetBone {boneName=dyn_jacket_b}}
    ]
    constraintWeights [1,1,1,1,1,1,1,1]
   }
   Output Pose(Result=Component To Local(ComponentPose=@n8))
   n9 = Control Rig!(Source=Copy Pose From Mesh() {bUseAttachedParent=true}, Alpha, IsInLOD2:bool=Get IsInLOD2() -> (IsInLOD2:bool)) -> (Pose) {
    alphaBoolBlend {blendOption=Linear}
    controlRigAssetReference {blueprintRigClass=&/M_MED_CarolinaChili_DeformationCtrlRig.M_MED_CarolinaChili_DeformationCtrlRig_C}
   }
   n10 = Get gravity_override() -> (gravity_override:Vector)
   n11 = RigidBody!(ComponentPose=Local To Component(LocalPose=@n9), Alpha, OverrideWorldGravity:Vector=@n10) -> (Pose) {
    bClampLinearTranslationLimitToRefPose=true
    bOverrideWorldGravity=true
    overridePhysicsAsset=&/M_MED_CarolinaChili_Physics
    simulationSpace=BaseBoneSpace
    alphaBoolBlend {blendOption=Linear}
    baseBoneRef {boneName=pelvis}
    componentAppliedLinearAccClamp=<3000 3000 3000>
    componentLinearAccScale=<0.45 0.45 0.45>
    componentLinearVelScale=<0.6 0.6 0.6>
    overrideWorldGravity {z=-980}
    simSpaceSettings {maxAngularAcceleration=500, maxAngularVelocity=3000, maxLinearAcceleration=500, maxLinearVelocity=3000, worldAlpha=0.7}
   }
   n12 = Control Rig!(Source=Component To Local(ComponentPose=@n11), Alpha) -> (Pose) {
    alphaBoolBlend {blendOption=Linear}
    controlRigAssetReference {blueprintRigClass=&/M_MED_CarolinaChili_CtrlRig.M_MED_CarolinaChili_CtrlRig_C}
   }
   n13 = RigidBody(ComponentPose=Local To Component(LocalPose=@n12), OverrideWorldGravity=@n10) {
    bClampLinearTranslationLimitToRefPose=true
    bOverrideWorldGravity=true
    overridePhysicsAsset=&/M_MED_CarolinaChili_Physics_Collar
    simulationSpace=BaseBoneSpace
    alphaBoolBlend {blendOption=Linear}
    baseBoneRef {boneName=spine_05}
    componentAppliedLinearAccClamp=<3000 3000 3000>
    componentLinearAccScale=<0.2 0.2 0.01>
    componentLinearVelScale=<0.6 0.6 0.6>
    overrideWorldGravity {z=-980}
    simSpaceSettings {maxAngularAcceleration=1000, maxAngularVelocity=3000, maxLinearAcceleration=1000, maxLinearVelocity=3000, worldAlpha=1}
   }
   n14 = "Constraint - dyn_zipper_r (PPPPPPPP)"!(
    ComponentPose=@n13
    Alpha=1
    ConstraintWeights_0=0.0189
    ConstraintWeights_1:float=0.0523
    ConstraintWeights_2:float=0.0748
    ConstraintWeights_3:float=0.0506
    ConstraintWeights_4:float=0.002
    ConstraintWeights_5:float=0.001
    ConstraintWeights_6:float=0.0013
    ConstraintWeights_7:float=0.7991
   ) -> (Pose) {
    alphaBoolBlend {blendOption=Linear}
    boneToModify {boneName=dyn_zipper_r}
    constraintSetup [
     {targetBone {boneName=spine_01}}
     {targetBone {boneName=spine_02}}
     {targetBone {boneName=spine_03}}
     {targetBone {boneName=spine_04}}
     {targetBone {boneName=spine_05}}
     {targetBone {boneName=clavicle_r}}
     {targetBone {boneName=deform_pec_r}}
     {targetBone {boneName=dyn_jacket_f_r}}
    ]
    constraintWeights [1,1,1,1,1,1,1,1]
   }
   n15 = "Constraint - dyn_zipper_l (PPPPPPPP)"!(
    ComponentPose=@n14
    Alpha=1
    ConstraintWeights_0=0.0604
    ConstraintWeights_1:float=0.0632
    ConstraintWeights_2:float=0.1039
    ConstraintWeights_3:float=0.0849
    ConstraintWeights_4:float=0.0012
    ConstraintWeights_5:float=0.0003
    ConstraintWeights_6:float=0.6835
    ConstraintWeights_7:float=0.0026
   ) -> (Pose) {
    alphaBoolBlend {blendOption=Linear}
    boneToModify {boneName=dyn_zipper_l}
    constraintSetup [
     {targetBone {boneName=spine_01}}
     {targetBone {boneName=spine_02}}
     {targetBone {boneName=spine_03}}
     {targetBone {boneName=spine_04}}
     {targetBone {boneName=spine_05}}
     {targetBone {boneName=clavicle_l}}
     {targetBone {boneName=dyn_jacket_f_l}}
     {targetBone {boneName=dyn_jacket_b}}
    ]
    constraintWeights [1,1,1,1,1,1,1,1]
   }
   n16 = Anim Dynamics - Bone: dyn_zipper_l(ComponentPose=@n15, Alpha=1, ExternalForce="0,0,0") {
    angularDampingOverride=0.9999
    angularSpringConstant=0.5
    bGravityOverrideInSimSpace=true
    bOverrideAngularDamping=true
    bOverrideLinearDamping=true
    linearDampingOverride=0.9999
    simulationSpace=RootRelative
    alphaBoolBlend {blendOption=Linear}
    boundBone {boneName=dyn_zipper_l}
    componentAppliedLinearAccClamp=<10000 10000 10000>
    componentLinearAccScale=<0.15 0.15 0.15>
    componentLinearVelScale=<0.15 0.15 0.15>
    gravityOverride=<1 1 1>
    physicsBodyDefinitions [{
      sphereCollisionRadius=0
      boundBone {boneName=dyn_zipper_l}
      boxExtents=<6 2 6>
      constraintSetup {angularLimitsMax=<5 30 -10>, angularLimitsMin=<-5 -30 -100>, angularTarget=<0 0 0>}
      localJointOffset=<0.3 0 0>
     }]
    relativeSpaceBone {boneName=root}
   }
   n17 = Anim Dynamics - Bone: dyn_zipper_r(ComponentPose=@n16, Alpha=1, ExternalForce="0,0,0") {
    angularDampingOverride=0.9999
    angularSpringConstant=0.5
    bGravityOverrideInSimSpace=true
    bOverrideAngularDamping=true
    bOverrideLinearDamping=true
    linearDampingOverride=0.9999
    simulationSpace=RootRelative
    alphaBoolBlend {blendOption=Linear}
    boundBone {boneName=dyn_zipper_r}
    componentAppliedLinearAccClamp=<10000 10000 10000>
    componentLinearAccScale=<0.15 0.15 0.15>
    componentLinearVelScale=<0.15 0.15 0.15>
    gravityOverride=<1 1 1>
    physicsBodyDefinitions [{
      sphereCollisionRadius=0
      boundBone {boneName=dyn_zipper_r}
      boxExtents=<6 2 6>
      constraintSetup {angularLimitsMax=<5 30 -10>, angularLimitsMin=<-5 -30 -100>, angularTarget=<0 0 0>}
      localJointOffset=<-0.3 0 0>
     }]
    relativeSpaceBone {boneName=root}
   }
   n18 = Anim Dynamics - Bone: deform_shoe_slap_c_l(ComponentPose=@n17, Alpha=1, ExternalForce="0,0,0") {
    angularDampingOverride=0.9999
    angularSpringConstant=0.5
    bGravityOverrideInSimSpace=true
    bLinearSpring=true
    bOverrideAngularDamping=true
    bOverrideLinearDamping=true
    linearDampingOverride=0.9999
    linearSpringConstant=150
    simulationSpace=RootRelative
    alphaBoolBlend {blendOption=Linear}
    alphaScaleBiasClamp {scale=0.4}
    boundBone {boneName=deform_shoe_slap_c_l}
    componentAppliedLinearAccClamp=<3000 3000 3000>
    componentLinearAccScale=<0.15 0.15 0.15>
    componentLinearVelScale=<0.15 0.15 0.15>
    gravityOverride=<1 1 1>
    physicsBodyDefinitions [{
      sphereCollisionRadius=0
      boundBone {boneName=deform_shoe_slap_c_l}
      boxExtents=<6 6 6>
      constraintSetup {angularLimitsMax=<0 0 0>, angularLimitsMin=<0 0 0>, angularTarget=<0 0 0>, linearAxesMax=<1 1 1>, linearAxesMin=<0 -5 -1>}
      localJointOffset=<-0.1 0 0>
     }]
    relativeSpaceBone {boneName=root}
   }
   n19 = Anim Dynamics - Bone: deform_shoe_slap_f_l(ComponentPose=@n18, Alpha=1, ExternalForce="0,0,0") {
    angularDampingOverride=0.9999
    angularSpringConstant=0.5
    bGravityOverrideInSimSpace=true
    bOverrideAngularDamping=true
    bOverrideLinearDamping=true
    linearDampingOverride=0.9999
    simulationSpace=RootRelative
    alphaBoolBlend {blendOption=Linear}
    alphaScaleBiasClamp {scale=0.5}
    boundBone {boneName=deform_shoe_slap_f_l}
    componentAppliedLinearAccClamp=<3000 3000 3000>
    componentLinearAccScale=<0.15 0.15 0.15>
    componentLinearVelScale=<0.15 0.15 0.15>
    gravityOverride=<1 1 1>
    physicsBodyDefinitions [{
      sphereCollisionRadius=0
      boundBone {boneName=deform_shoe_slap_f_l}
      boxExtents=<6 6 2>
      constraintSetup {angularLimitsMax=<5 20 15>, angularLimitsMin=<-5 -35 -15>, angularTarget=<0 0 0>}
      localJointOffset=<3 0 0>
     }]
    relativeSpaceBone {boneName=root}
   }
   n20 = Anim Dynamics - Bone: deform_shoe_slap_c_r(ComponentPose=@n19, Alpha=1, ExternalForce="0,0,0") {
    angularDampingOverride=0.9999
    angularSpringConstant=0.5
    bGravityOverrideInSimSpace=true
    bLinearSpring=true
    bOverrideAngularDamping=true
    bOverrideLinearDamping=true
    linearDampingOverride=0.9999
    linearSpringConstant=150
    simulationSpace=RootRelative
    alphaBoolBlend {blendOption=Linear}
    alphaScaleBiasClamp {scale=0.4}
    boundBone {boneName=deform_shoe_slap_c_r}
    componentAppliedLinearAccClamp=<3000 3000 3000>
    componentLinearAccScale=<0.15 0.15 0.15>
    componentLinearVelScale=<0.15 0.15 0.15>
    gravityOverride=<1 1 1>
    physicsBodyDefinitions [{
      sphereCollisionRadius=0
      boundBone {boneName=deform_shoe_slap_c_r}
      boxExtents=<6 6 6>
      constraintSetup {angularLimitsMax=<0 0 0>, angularLimitsMin=<0 0 0>, angularTarget=<0 0 0>, linearAxesMax=<0 5 1>, linearAxesMin=<-1 -1 -1>}
      localJointOffset=<0.1 0 0>
     }]
    relativeSpaceBone {boneName=root}
   }
   n21 = Anim Dynamics - Bone: deform_shoe_slap_f_r(ComponentPose=@n20, Alpha=1, ExternalForce="0,0,0") {
    angularDampingOverride=0.9999
    angularSpringConstant=0.5
    bGravityOverrideInSimSpace=true
    bOverrideAngularDamping=true
    bOverrideLinearDamping=true
    linearDampingOverride=0.9999
    simulationSpace=RootRelative
    alphaBoolBlend {blendOption=Linear}
    alphaScaleBiasClamp {scale=0.5}
    boundBone {boneName=deform_shoe_slap_f_r}
    componentAppliedLinearAccClamp=<3000 3000 3000>
    componentLinearAccScale=<0.15 0.15 0.15>
    componentLinearVelScale=<0.15 0.15 0.15>
    gravityOverride=<1 1 1>
    physicsBodyDefinitions [{
      sphereCollisionRadius=0
      boundBone {boneName=deform_shoe_slap_f_r}
      boxExtents=<6 6 2>
      constraintSetup {angularLimitsMax=<5 35 15>, angularLimitsMin=<-5 -20 -15>, angularTarget=<0 0 0>}
      localJointOffset=<-3 0 0>
     }]
    relativeSpaceBone {boneName=root}
   }
   Save cached pose 'Base Pose'(Pose=Component To Local(ComponentPose=@n21)) {cacheName=Base Pose}
   n22 = "Constraint - dyn_jacket_b (R)"(ComponentPose=Local To Component(LocalPose=Use cached pose 'Base Pose'()), Alpha=1, ConstraintWeights_0=1) {
    alphaBoolBlend {blendOption=Linear}
    alphaScaleBiasClamp {scale=0.8}
    boneToModify {boneName=dyn_jacket_b}
    constraintSetup [{transformType=Rotation, targetBone {boneName=spine_02}}]
    constraintWeights [1]
   }
   n23 = Look At - Bone: dyn_jacket_b(ComponentPose=@n22, Alpha=1) {
    alphaBoolBlend {blendOption=Linear}
    boneToModify {boneName=dyn_jacket_b}
    lookAtLocation=<0 -7>
    lookAtTarget {boneReference {boneName=spine_02}}
    lookAt_Axis {axis=<1 0>}
    lookUp_Axis {axis=<0 -1>}
   }
   n24 = Component To Local(ComponentPose=@n23)
   n25 = Use cached pose 'Base Pose'()
   n26 = Get isWearingPack() -> (isWearingPack:bool)
   n27 = "Transform (Modify) Bone - Bone: dyn_collar_b"!(
    ComponentPose=Local To Component(LocalPose=Blend Poses by bool(BlendPose_0=@n24, BlendPose_1=@n25, bActiveValue=@n26) {blendType=Linear})
    bAlphaBoolEnabled:bool=Get bIsSkyDiving()
    Translation="0,0,0"
    Rotation="0,0,0"
    Scale="1,1,0.8"
   ) -> (Pose) {
    alphaInputType=Bool
    scaleMode=BMM_Additive
    scaleSpace=BCS_BoneSpace
    alphaBoolBlend {blendInTime=0.1, blendOption=Linear, blendOutTime=0.1}
    alphaScaleBiasClamp {bInterpResult=true}
    boneToModify {boneName=dyn_collar_b}
    scale {z=0.8}
   }
   n28 = "Transform (Modify) Bone - Bone: head"!(
    ComponentPose=@n27
    bAlphaBoolEnabled:bool=Get EmoteNotPlaying() -> (EmoteNotPlaying:bool)
    Translation="0,0,0"
    Rotation=Get head rot() -> (head rot:Rotator)
    Scale="1,1,1"
   ) -> (Pose) {
    alphaInputType=Bool
    rotationMode=BMM_Additive
    rotationSpace=BCS_BoneSpace
    scaleMode=BMM_Additive
    scaleSpace=BCS_BoneSpace
    alphaBoolBlend {blendOption=Linear}
    boneToModify {boneName=head}
   }
   Save cached pose 'prefrontend'(Pose=Component To Local(ComponentPose=@n28)) {cacheName=prefrontend}
  ]}
 EventGraph {nodes [
   Event BlueprintUpdateAnimation()
   > n1 = Gravity Override State Function DEPRECATED(CharacterPartAnimInstance=Self-Reference(), GameTimeInSeconds=GetGameTimeInSeconds())
   > Set gravity_override(gravity_override=@n1)
   n2 = Get Mesh(self=Get Pawn Owner() -> (Pawn Owner:Player Pawn Athena Object Reference))
   n3 = BreakRotator(InRot=BreakTransform(InTransform=GetDeltaTransformFromRefPose(self=@n2, BoneName=head, BaseName=neck_02)).Rotation)
   n4 = Get Y percentage() -> (Y percentage:float)
   n5 = Get X percentage() -> (X percentage:float)
   n6 = Get Z percentage() -> (Z percentage:float)
   > Set head rot(head rot:Rotator=MakeRotator(
     Roll=float * float!(A=@n3, B=-1, C:double=@n5) -> (ReturnValue)
     Pitch=float * float(A=@n3.Pitch, B=-1, C=@n4)
     Yaw=float * float(A=@n3.Yaw, B=-1, C=@n6)
    )) -> (Output_Get:Rotator)
   > Set EmoteNotPlaying(EmoteNotPlaying:bool=NOT Boolean(A=Get bIsPlayingEmote())) -> (Output_Get:bool)
   > Set IsInLOD2(IsInLOD2:bool="integer > integer"(A=Get CurrentLOD(), B=1)) -> (Output_Get:bool)
   Event BlueprintInitializeAnimation()
   > n7 = Cast To PlayerPawn_Athena(Object=TryGetPawnOwner())
   > n8 = Set Pawn Owner(Pawn Owner:Player Pawn Athena Object Reference=@n7) -> (Output_Get:Player Pawn Athena Object Reference)
   > Set Head_Mesh(Head_Mesh:Skeletal Mesh Component Object Reference=GetSkeletalMeshForPartType(self=@n8, PartType=Head)) -> (Output_Get:Skeletal Mesh Component Object Reference)
   Event OnCharacterCustomizationLoaded()
   n9 = CheckForBackbling()
   > Set isWearingPack(isWearingPack:bool=@n9) -> (Output_Get:bool)
   > Set isWearingCape(isWearingCape=AND Boolean(A=@n9, B=@n9.bIsBackblingLong))
   TryGetPawnOwner()
  ]}
]
variables [
 EmoteNotPlaying {type=boolean}
 gravity_override {type=Vector}
 head rot {type=Rotator}
 Head_Mesh {type=SkeletalMeshComponent}
 IsInLOD2 {type=boolean}
 isWearingCape {type=boolean}
 isWearingPack {type=boolean}
 Pawn Owner {type=PlayerPawn_Athena_C}
 X percentage {type=number, default=0.7}
 Y percentage {type=number, default=0.7}
 Z percentage {type=number, default=0.7}
]
